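# Masked skeletal-mesh material for the UnluckyRoll character body, built as a spherical dissolve with a hot edge for the Kaiju Demon defeat FX.
# Sphere masks grow from a fixed point in pre-skinned local space by the Time parameter, with noise from a tiled texture breaking up their borders.
# An inverted, offset-radius mask tints the hot edge pink-red.
# Body textures are sampled, but the graph shows no links to material outputs.
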Material /GrillMoon/Assets/FX/Kaiju/Fight/Demon/Defeat/Materials/M_GrillMoon_UnluckyRoll_Dissolve
graph {
 properties {
  bAutomaticallySetUsageInEditor=false
  bUsedWithMeshDeformer=true
  bUsedWithSkeletalMesh=true
  blendMode=BLEND_Masked
  usedShadingModels=Default Lit
 }
 nodes [
  ConstantBiasScale(None=MaterialFunctionCall() {materialFunction=&BoundingBoxBased_0-1_UVW}.RGB) {bias=-0.5, scale=2}
  n1 = TextureSample() {samplerType=SAMPLERTYPE_Masks, texture=/BRCosmetics/Characters/Player/Male/Large/Bodies/M_LRG_UnluckyRoll/Textures/T_UnluckyRoll_Body_S}
  LinearInterpolate(Alpha=@n1.R)
  LinearInterpolate(Alpha=@n1.B)
  LinearInterpolate(Alpha=@n1.G)
  n2 = TextureSample(UVs=Multiply(A=TextureCoordinate(), B=ScalarParameter() {defaultValue=10, parameterName=Param_1})) {texture=&2ndInfectedMapLinear}
  n3 = Add(
   A=Multiply(A=@n2.G, B=ScalarParameter() {defaultValue=4.0099, parameterName=hot edge distortion})
   B=VertexInterpolator(VS=LocalPosition() {localOrigin=InstancePreSkinning}.XYZ).PS
  )
  n4 = VectorParameter() {defaultValue=<rgba 8.4153 18.8901 126.3204 0>}
  n5 = ScalarParameter() {defaultValue=4.9801, parameterName=Time}
  n6 = Add(A=@n5, B=ScalarParameter() {defaultValue=-15.1389, parameterName=Param_4})
  n7 = ScalarParameter() {defaultValue=0.7093, parameterName=hot edge hardness}
  Multiply(
   A=VectorParameter() {parameterName=Param_1, defaultValue=<rgba 5 0 1.9414 1>}.RGB
   B=OneMinus(None=SphereMask(A=@n3, B=@n4.RGB, Radius=@n6, Hardness=@n7))
  )
  ScalarParameter() {defaultValue=59.9046, parameterName=Radius}
  ScalarParameter() {defaultValue=5, parameterName=Param_3}
  n8 = ScalarParameter() {defaultValue=8, parameterName=Param_2}
  n9 = LocalPosition() {localOrigin=InstancePreSkinning}
  SphereMask(A=Add(A=Multiply(A=@n2.G, B=@n8), B=VertexInterpolator(VS=@n9.XYZ).PS), B=@n4.RGB, Radius=@n5, Hardness=ScalarParameter())
  TextureSample() {texture=/BRCosmetics/Characters/Player/Male/Large/Bodies/M_LRG_UnluckyRoll/Textures/T_UnluckyRoll_Body_D}
  TextureSample() {samplerType=SAMPLERTYPE_Masks, texture=/BRCosmetics/Characters/Player/Male/Large/Bodies/M_LRG_UnluckyRoll/Textures/T_UnluckyRoll_Body_M}
  TextureSample() {samplerType=SAMPLERTYPE_Normal, texture=/BRCosmetics/Characters/Player/Male/Large/Bodies/M_LRG_UnluckyRoll/Textures/T_UnluckyRoll_Body_N}
 ]
}
