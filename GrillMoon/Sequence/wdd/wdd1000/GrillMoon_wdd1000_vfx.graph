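# VFX Level Sequence for cinematic wdd1000, showing the death of a demon character (GrillMoon_UnluckyRoll) and a portal overload and explosion.
# Attached Niagara slash, chest smoke, beam and ash effects play while the Body mesh swaps to a dissolve material with an animated Time parameter.
# Keyed Niagara parameters steer the beam direction and ramp the overload's scale and lightning.
# As the portal explosion spawns, a post-process volume spikes exposure compensation and intensity, then fades.
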LevelSequence /GrillMoon/Sequence/wdd/wdd1000/GrillMoon_wdd1000_vfx
bindings [
 NiagaraComponent0 {
  id=#1
  parent_id=#2
  possessed_class=NiagaraComponent
  tracks [
   {type=MovieSceneNiagaraSystemTrack, sections [{type=MovieSceneNiagaraSystemSpawnSection, end_frame=4111, start_frame=3496}]}
   {
    type=MovieSceneNiagaraVectorParameterTrack
    display_name=True Vector_Direction End
    sections [{
      type=MovieSceneFloatVectorSection
      channels [
       Vector.X {type=MovieSceneScriptingFloatChannel, default=39331.5547, keys [[3451,15400.4141],[3674,-58.4119],[3811,39331.5547]]}
       Vector.Y {type=MovieSceneScriptingFloatChannel, default=-52323.1641, keys [[3445,-64607.7734],[3674,-77042.2969],[3811,-52323.1641]]}
       Vector.Z {type=MovieSceneScriptingFloatChannel, default=23141.9062, keys [[3445,0.2962],[3811,-4.1356],[4049,-595.8926],[4095,32677.9062]]}
      ]
     }]
   }
   {
    type=MovieSceneNiagaraFloatParameterTrack
    display_name=Head Lifetime
    sections [{type=MovieSceneFloatSection, channels [None {type=MovieSceneScriptingFloatChannel, default=0.25, keys [[3501,1],[4020,0.25]]}]}]
   }
   {
    type=MovieSceneNiagaraFloatParameterTrack
    display_name=Head Spawn Rate
    sections [{type=MovieSceneFloatSection, channels [None {type=MovieSceneScriptingFloatChannel, default=0, keys [[3496,1],[3963,5],[4020,0]]}]}]
   }
  ]
 }
 NiagaraComponent0 {
  id=#3
  parent_id=#4
  possessed_class=NiagaraComponent
  tracks [{type=MovieSceneNiagaraSystemTrack, sections [{type=MovieSceneNiagaraSystemSpawnSection, end_frame=4179, start_frame=3430}]}]
 }
 PostProcessVolume {
  id=#5
  possessed_class=PostProcessVolume
  tracks [
   {type=MovieSceneSpawnTrack, sections [{type=MovieSceneSpawnSection, channels [None {type=MovieSceneScriptingBoolChannel, default=true}]}]}
   {
    type=MovieScene3DTransformTrack
    sections [{
      type=MovieScene3DTransformSection
      channels [
       Location.X {type=MovieSceneScriptingDoubleChannel, default=0}
       Location.Y {type=MovieSceneScriptingDoubleChannel, default=0}
       Location.Z {type=MovieSceneScriptingDoubleChannel, default=0}
       Rotation.X {type=MovieSceneScriptingDoubleChannel, default=0}
       Rotation.Y {type=MovieSceneScriptingDoubleChannel, default=0}
       Rotation.Z {type=MovieSceneScriptingDoubleChannel, default=0}
       Scale.X {type=MovieSceneScriptingDoubleChannel, default=1}
       Scale.Y {type=MovieSceneScriptingDoubleChannel, default=1}
       Scale.Z {type=MovieSceneScriptingDoubleChannel, default=1}
      ]
     }]
   }
   {
    type=MovieSceneFloatTrack
    display_name="Exposure Compensation (Settings)"
    sections [{type=MovieSceneFloatSection, channels [None {type=MovieSceneScriptingFloatChannel, default=20, keys [[4095.69,0],[4111.15,20],[4273,0]]}]}]
   }
   {
    type=MovieSceneFloatTrack
    display_name="Intensity (Settings)"
    sections [{type=MovieSceneFloatSection, channels [None {type=MovieSceneScriptingFloatChannel, default=0, keys [[4097.5875,0],[4111.15,1],[4273,0]]}]}]
   }
  ]
 }
 NS_Death_PortalOverload {
  id=#6
  possessed_class=NiagaraActor
  tracks [
   {type=MovieSceneSpawnTrack, sections [{type=MovieSceneSpawnSection, channels [None {type=MovieSceneScriptingBoolChannel, default=true}]}]}
   {
    type=MovieScene3DTransformTrack
    sections [{
      type=MovieScene3DTransformSection
      channels [
       Location.X {type=MovieSceneScriptingDoubleChannel, default=-57602.0454}
       Location.Y {type=MovieSceneScriptingDoubleChannel, default=42804.5333}
       Location.Z {type=MovieSceneScriptingDoubleChannel, default=10457.2506}
       Rotation.X {type=MovieSceneScriptingDoubleChannel, default=0}
       Rotation.Y {type=MovieSceneScriptingDoubleChannel, default=0}
       Rotation.Z {type=MovieSceneScriptingDoubleChannel, default=0}
       Scale.X {type=MovieSceneScriptingDoubleChannel, default=1}
       Scale.Y {type=MovieSceneScriptingDoubleChannel, default=1}
       Scale.Z {type=MovieSceneScriptingDoubleChannel, default=1}
      ]
     }]
   }
  ]
 }
 NiagaraComponent0 {
  id=#7
  parent_id=#6
  possessed_class=NiagaraComponent
  tracks [
   {
    type=MovieSceneNiagaraFloatParameterTrack
    display_name=Uniform Scale Factor
    sections [{
      type=MovieSceneFloatSection
      channels [None {type=MovieSceneScriptingFloatChannel, default=12, keys [[3693,0],[3710,0.5],[4086.8225,12],[4086.955,10],[4091.1025,1],[4094,10],[4251,0]]}]
     }]
   }
   {
    type=MovieSceneNiagaraFloatParameterTrack
    display_name=B_SpawnRate
    sections [{
      type=MovieSceneFloatSection
      channels [None {type=MovieSceneScriptingFloatChannel, default=0, keys [[3694,0],[3748,0.25],[4086,1.5],[4087,1],[4091,0]]}]
     }]
   }
   {
    type=MovieSceneNiagaraFloatParameterTrack
    display_name=LightningLifetime
    sections [{type=MovieSceneFloatSection, channels [None {type=MovieSceneScriptingFloatChannel, default=0.5, keys [[3693,1],[4073,0.5]]}]}]
   }
   {
    type=MovieSceneNiagaraFloatParameterTrack
    display_name=LightningScale
    sections [{type=MovieSceneFloatSection, channels [None {type=MovieSceneScriptingFloatChannel, default=1.5, keys [[3693,0.5],[4085,1.5]]}]}]
   }
  ]
 }
 NS_ChestSmoke {
  id=#8
  possessed_class=NiagaraActor
  tracks [
   {type=MovieSceneSpawnTrack, sections [{type=MovieSceneSpawnSection, channels [None {type=MovieSceneScriptingBoolChannel, default=true}]}]}
   {
    type=MovieScene3DTransformTrack
    sections [{
      type=MovieScene3DTransformSection
      channels [
       Location.X {type=MovieSceneScriptingDoubleChannel, default=14.7799}
       Location.Y {type=MovieSceneScriptingDoubleChannel, default=1.9984}
       Location.Z {type=MovieSceneScriptingDoubleChannel, default=-5.5401}
       Rotation.X {type=MovieSceneScriptingDoubleChannel, default=11.0133}
       Rotation.Y {type=MovieSceneScriptingDoubleChannel, default=-4.5093}
       Rotation.Z {type=MovieSceneScriptingDoubleChannel, default=10.956}
       Scale.X {type=MovieSceneScriptingDoubleChannel, default=1}
       Scale.Y {type=MovieSceneScriptingDoubleChannel, default=1}
       Scale.Z {type=MovieSceneScriptingDoubleChannel, default=1}
      ]
     }]
   }
   {type=MovieScene3DAttachTrack, sections [{type=MovieScene3DAttachSection, end_frame=12085, start_frame=3280}]}
  ]
 }
 NiagaraComponent0 {
  id=#9
  parent_id=#10
  possessed_class=NiagaraComponent
  tracks [{type=MovieSceneNiagaraSystemTrack, sections [{type=MovieSceneNiagaraSystemSpawnSection, end_frame=4691, start_frame=4091}]}]
 }
 NS_DemonAsh_Mask {
  id=#11
  possessed_class=NiagaraActor
  tracks [
   {
    type=MovieSceneSpawnTrack
    sections [{type=MovieSceneSpawnSection, channels [None {type=MovieSceneScriptingBoolChannel, default=false, keys [[4179,true],[4196,false]]}]}]
   }
   {
    type=MovieScene3DTransformTrack
    sections [{
      type=MovieScene3DTransformSection
      channels [
       Location.X {type=MovieSceneScriptingDoubleChannel, default=-548.5329, keys [[3433,-428.1304],[3572,-406.018],[3760,-406.018],[4012,-495.4854]]}
       Location.Y {type=MovieSceneScriptingDoubleChannel, default=-36.278, keys [[3433,104.2364],[3572,90.0631],[3760,90.0631],[4012,66.7448]]}
       Location.Z {type=MovieSceneScriptingDoubleChannel, default=71.0987, keys [[3433,81.8382],[3572,96.1822],[3760,141.0388],[4012,174.4816]]}
       Rotation.X {type=MovieSceneScriptingDoubleChannel, default=0, const [0,4]}
       Rotation.Y {type=MovieSceneScriptingDoubleChannel, default=0, const [0,4]}
       Rotation.Z {type=MovieSceneScriptingDoubleChannel, default=14.6083, const [14.6083,4]}
       Scale.X {type=MovieSceneScriptingDoubleChannel, default=1, const [1,4]}
       Scale.Y {type=MovieSceneScriptingDoubleChannel, default=1, const [1,4]}
       Scale.Z {type=MovieSceneScriptingDoubleChannel, default=1, const [1,4]}
      ]
     }]
   }
   {type=MovieScene3DAttachTrack, sections [{type=MovieScene3DAttachSection, end_frame=11870, start_frame=3402}]}
  ]
 }
 NiagaraComponent0 {
  id=#12
  parent_id=#11
  possessed_class=NiagaraComponent
  tracks [{type=MovieSceneNiagaraSystemTrack, sections [{type=MovieSceneNiagaraSystemSpawnSection, end_frame=4124, start_frame=3430}]}]
 }
 NiagaraComponent0 {
  id=#13
  parent_id=#8
  possessed_class=NiagaraComponent
  tracks [
   {type=MovieSceneNiagaraSystemTrack, sections [{type=MovieSceneNiagaraSystemSpawnSection, end_frame=4158, start_frame=3437}]}
   {
    type=MovieSceneNiagaraFloatParameterTrack
    display_name=Emitter_ErodeSmoke
    sections [{
      type=MovieSceneFloatSection
      channels [None {type=MovieSceneScriptingFloatChannel, default=0.1, keys [[3437,-1],[3439,-2],[3759,0],[3793,0.1]]}]
     }]
   }
  ]
 }
 NS_ChestBeam_ToPortal {
  id=#2
  possessed_class=NiagaraActor
  tracks [
   {
    type=MovieSceneSpawnTrack
    sections [{type=MovieSceneSpawnSection, channels [None {type=MovieSceneScriptingBoolChannel, default=true, const [true,1]}]}]
   }
   {
    type=MovieScene3DTransformTrack
    sections [{
      type=MovieScene3DTransformSection
      channels [
       Location.X {type=MovieSceneScriptingDoubleChannel, default=0, const [0,1]}
       Location.Y {type=MovieSceneScriptingDoubleChannel, default=0.8806, const [0.8806,1]}
       Location.Z {type=MovieSceneScriptingDoubleChannel, default=-17.1237, const [-17.1237,1]}
       Rotation.X {type=MovieSceneScriptingDoubleChannel, default=0, const [0,1]}
       Rotation.Y {type=MovieSceneScriptingDoubleChannel, default=0, const [0,1]}
       Rotation.Z {type=MovieSceneScriptingDoubleChannel, default=0, const [0,1]}
       Scale.X {type=MovieSceneScriptingDoubleChannel, default=1, const [1,1]}
       Scale.Y {type=MovieSceneScriptingDoubleChannel, default=1, const [1,1]}
       Scale.Z {type=MovieSceneScriptingDoubleChannel, default=1, const [1,1]}
      ]
     }]
   }
   {type=MovieScene3DAttachTrack, sections [{type=MovieScene3DAttachSection, end_frame=11953, start_frame=3431}]}
  ]
 }
 NS_DemonSlashed {
  id=#4
  possessed_class=NiagaraActor
  tracks [
   {type=MovieSceneSpawnTrack, sections [{type=MovieSceneSpawnSection, channels [None {type=MovieSceneScriptingBoolChannel, default=true}]}]}
   {
    type=MovieScene3DTransformTrack
    sections [{
      type=MovieScene3DTransformSection
      channels [
       Location.X {type=MovieSceneScriptingDoubleChannel, default=4.7581, const [4.7581,1]}
       Location.Y {type=MovieSceneScriptingDoubleChannel, default=6.6799, const [6.6799,1]}
       Location.Z {type=MovieSceneScriptingDoubleChannel, default=-10.8388, const [-10.8388,1]}
       Rotation.X {type=MovieSceneScriptingDoubleChannel, default=-56.573, const [-56.573,1]}
       Rotation.Y {type=MovieSceneScriptingDoubleChannel, default=1.3141, const [1.3141,1]}
       Rotation.Z {type=MovieSceneScriptingDoubleChannel, default=-8.5799, const [-8.5799,1]}
       Scale.X {type=MovieSceneScriptingDoubleChannel, default=0.01, const [0.01,1]}
       Scale.Y {type=MovieSceneScriptingDoubleChannel, default=0.01, const [0.01,1]}
       Scale.Z {type=MovieSceneScriptingDoubleChannel, default=0.0115, const [0.0181,1]}
      ]
     }]
   }
   {type=MovieScene3DAttachTrack, sections [{type=MovieScene3DAttachSection, end_frame=11841, start_frame=3430}]}
  ]
 }
 Body {
  id=#14
  parent_id=#15
  possessed_class=SkeletalMeshComponentBudgeted
  tracks [
   {
    type=MovieScenePrimitiveMaterialTrack
    display_name=Material Slot: Body
    sections [{
      type=MovieScenePrimitiveMaterialSection
      channels [None {
        type=MovieSceneScriptingObjectPathChannel
        default="<Object '/GrillMoon/Assets/FX/Kaiju/Fight/Demon/Defeat/Materials/M_GrillMoon_UnluckyRoll_Dissolve' (0x850b543e000) Class 'Material'>"
        const ["<Object '&M_GrillMoon_UnluckyRoll_Dissolve' (0x850b543e000) Class 'Material'>", 1]
       }]
     }]
   }
   {
    type=MovieSceneComponentMaterialTrack
    display_name=Material Slot: Body
    sections [{
      type=MovieSceneComponentMaterialParameterSection
      end_frame=12085
      start_frame=0
      channels [Time {type=MovieSceneScriptingFloatChannel, default=55, keys [[3415,200],[4122,55]]}]
     }]
   }
  ]
 }
 GrillMoon_UnluckyRoll {id=#15, possessed_class=FortPlayerMannequin, tracks [{type=MovieSceneSkeletalAnimationTrack, sections []}]}
 NS_Death_Portal_Explode {
  id=#10
  possessed_class=NiagaraActor
  tracks [
   {type=MovieSceneSpawnTrack, sections [{type=MovieSceneSpawnSection, channels [None {type=MovieSceneScriptingBoolChannel, default=true}]}]}
   {
    type=MovieScene3DTransformTrack
    sections [{
      type=MovieScene3DTransformSection
      channels [
       Location.X {type=MovieSceneScriptingDoubleChannel, default=-57712.9619}
       Location.Y {type=MovieSceneScriptingDoubleChannel, default=42804.5333}
       Location.Z {type=MovieSceneScriptingDoubleChannel, default=10548.4306}
       Rotation.X {type=MovieSceneScriptingDoubleChannel, default=0}
       Rotation.Y {type=MovieSceneScriptingDoubleChannel, default=0}
       Rotation.Z {type=MovieSceneScriptingDoubleChannel, default=0}
       Scale.X {type=MovieSceneScriptingDoubleChannel, default=1}
       Scale.Y {type=MovieSceneScriptingDoubleChannel, default=1}
       Scale.Z {type=MovieSceneScriptingDoubleChannel, default=1}
      ]
     }]
   }
  ]
 }
]
display_rate {denominator=1, numerator=60}
tick_resolution {denominator=1, numerator=24000}
tracks []
